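# Unlit translucent Niagara ribbon material for the InstantGravel character's FX, drawing colored, flame-shaped ribbons with noisy dissolve.
# Particle color RGB feeds the emissive adjustments function, and particle alpha scales the opacity mask.
# A wispy fire noise, panned on UV1 with a dynamic U offset and scaled by dynamic Dissolve, erodes an inverted UV0 gradient.
# Parameterized smoothsteps soften the base, tip and sides, while Camera Offset and HueShift are named but unused.
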Material /Game/Characters/Player/Male/Medium/Bodies/M_MED_InstantGravel/FX/M_InstantGravel_ColoredRibbons
graph {
 properties {bUsedWithNiagaraRibbons=true, blendMode=BLEND_Translucent, pixelDepthOffsetMode=PDOM_Legacy, shadingModel=MSM_Unlit, usedShadingModels=Unlit}
 nodes [
  n1 = ParticleColor()
  MaterialFunctionCall(Emissive=@n1.RGB) {materialFunction=&MF_FX_EmissiveAdjustments}
  n2 = OneMinus(None=TextureCoordinate())
  n3 = DynamicParameter() {defaultValue=<rgba 0.5 0 0 0>, paramNames [UV1 U Offset, Dissolve, Camera Offset, HueShift]}
  n4 = Constant()
  n5 = TextureCoordinate() {coordinateIndex=1}
  n6 = VectorParameter() {group=06- Distortion, parameterName=DistoUV_Settings, channelNames {a=PanY, b=PanX, g=ScaleY, r=ScaleX}, defaultValue=<rgba 0.8 0.15 -2 0>}
  n7 = AppendVector(A=@n6.ScaleX, B=@n6.ScaleX)
  n8 = Add(A=Add(A=AppendVector(A=@n3.UV1 U Offset, B=@n4), B=Multiply(A=@n5, B=@n7)), B=Multiply(A=Time(), B=AppendVector(A=@n6.PanX, B=@n6.PanX)))
  n9 = Multiply(A=SmoothStep(Value=TextureSample(UVs=@n8) {samplerType=SAMPLERTYPE_Grayscale, texture=/Game/Effects/Fort_Effects/Textures/Fire/T_WispyFireNoise_Soft_M}.R) {constMax=0.4}) {constB=2}
  n10 = DynamicParameter() {defaultValue=<rgba 0.5 0 0 0>, paramNames [UV1 U Offset, Dissolve, Camera Offset, HueShift]}
  n11 = ComponentMask(None=Subtract(A=Add(A=@n2), B=Multiply(A=@n9, B=@n10.Dissolve))) {r=true}
  n12 = TextureCoordinate()
  n13 = ComponentMask(None=@n12) {r=true}
  n14 = SmoothStep(Max=ScalarParameter() {defaultValue=0.02, group=04- Shape, parameterName=FlameBaseSoftness}, Value=@n13) {constMax=0.2}
  n15 = SmoothStep(Max=ScalarParameter() {defaultValue=0.5, group=04- Shape, parameterName=FlameTipSoftness}, Value=OneMinus(None=@n13)) {constMax=0.1}
  n16 = ScalarParameter() {defaultValue=0.4, group=04- Shape, parameterName=FlameSideSoftness}
  n17 = ComponentMask(None=@n12) {g=true}
  n18 = Multiply(A=@n17, B=OneMinus(None=@n17))
  Multiply(A=Multiply(A=@n1.A, B=Saturate(None=@n11)), B=Multiply(A=Multiply(A=@n14, B=@n15), B=SmoothStep(Max=@n16, Value=@n18) {constMax=0.5}))
 ]
}
