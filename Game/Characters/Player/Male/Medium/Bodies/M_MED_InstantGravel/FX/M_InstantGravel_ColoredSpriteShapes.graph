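# Translucent unlit Niagara sprite material for colored shape particles in the Instant Gravel character's effects.
# Particle color RGB feeds a shared FX emissive adjustment function.
# Particle alpha is multiplied by the red channel of a swappable texture parameter, defaulting to a shape flipbook, to form a shape mask.
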Material /Game/Characters/Player/Male/Medium/Bodies/M_MED_InstantGravel/FX/M_InstantGravel_ColoredSpriteShapes
graph {
 properties {bUsedWithNiagaraSprites=true, blendMode=BLEND_Translucent, pixelDepthOffsetMode=PDOM_Legacy, shadingModel=MSM_Unlit, usedShadingModels=Unlit}
 nodes [
  n1 = ParticleColor()
  MaterialFunctionCall(Emissive=@n1.RGB) {materialFunction=&MF_FX_EmissiveAdjustments}
  n2 = TextureSampleParameter2D() {
   parameterName=Texture
   samplerType=SAMPLERTYPE_LinearColor
   texture=&/T_InstantGravel_ShapeFlipBook
  }
  Multiply(A=@n1.A, B=@n2.R)
 ]
}
